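# Sound cue for the replay (DVR) camera zoom-out sound, routed to the non-critical in-game UI sound class.
# A Mixer layers a looping zoom-out wave at half volume with a one-shot zoom-out wave at full volume.
# An Enveloper on the loop applies a falling pitch curve from slightly above to slightly below normal, while its volume curve stays flat.
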SoundCue /Game/Sounds/Fort_UI_Sounds/Replay/DVR_Cam_ZoomOutLoop_Cue
root=Mixer_0
nodes [
 {
  id=Enveloper_0
  children [WavePlayer_0]
  properties {pitchCurve {editorCurveData {keys [<at 0 1.1>,<at 1 0.9>]}}, volumeCurve {editorCurveData {keys [<at 0 1>,<at 0.1 1>]}}}
 }
 {id=Mixer_0, children [Enveloper_0,WavePlayer_1], properties {inputVolume [0.5,1]}}
 {id=WavePlayer_0, properties {bLooping=true, soundWaveAssetPtr=&/DVR_Cam_ZoomOutLoop}}
 {id=WavePlayer_1, properties {soundWaveAssetPtr=&/DVR_Cam_ZoomOut}}
]
properties {maxDistance=2097152, soundClassObject=&Fort_UI_InGame_NonCritical}
