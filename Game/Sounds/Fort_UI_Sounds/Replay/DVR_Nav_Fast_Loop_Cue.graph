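# Sound cue for a looping fast-navigation UI sound in the replay (DVR) interface, assigned to the non-critical in-game UI sound class.
# A mixer blends three looping wave layers at different input levels, each shaped by an indefinitely looping envelope that pulses its volume.
# The third layer's envelope also sweeps pitch and runs on a shorter cycle than the others, then passes through a modulator with a fixed pitch raise.
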SoundCue /Game/Sounds/Fort_UI_Sounds/Replay/DVR_Nav_Fast_Loop_Cue
root=Mixer_0
nodes [
 {
  id=Enveloper_3
  children [WavePlayer_1]
  properties {
   bLoop=true
   bLoopIndefinitely=true
   loopEnd=20
   pitchCurve {editorCurveData {keys [{time=0}]}}
   volumeCurve {editorCurveData {keys [<at 0 1>,<at 5 0.5>,<at 15 0.5>,<at 20 1>]}}
  }
 }
 {
  id=Enveloper_4
  children [WavePlayer_3]
  properties {
   bLoop=true
   bLoopIndefinitely=true
   loopEnd=17
   pitchCurve {editorCurveData {keys [{time=0}, <at 10 0.5>, {time=17}]}}
   volumeCurve {editorCurveData {keys [<at 0 0.2>,<at 3 0.5>,<at 6 0.2>,<at 10 0.2>,<at 15 0.5>,<at 17 0.2>]}}
  }
 }
 {
  id=Enveloper_5
  children [WavePlayer_2]
  properties {
   bLoop=true
   bLoopIndefinitely=true
   loopEnd=20
   pitchCurve {editorCurveData {keys [{time=0}]}}
   volumeCurve {editorCurveData {keys [<at 0 1>,<at 5 0.75>,<at 10 1>,<at 15 0.75>,<at 20 1>]}}
  }
 }
 {id=Mixer_0, children [Enveloper_3,Modulator_1,Enveloper_5], properties {inputVolume [2.5,1,1.5]}}
 {id=Modulator_1, children [Enveloper_4], properties {pitchMax=2, pitchMin=2}}
 {id=WavePlayer_1, properties {bLooping=true, soundWaveAssetPtr=&/DVR_Nav_Fast_Loop_Layer_01}}
 {id=WavePlayer_2, properties {bLooping=true, soundWaveAssetPtr=&/DVR_Nav_Fast_Loop_Layer_02}}
 {id=WavePlayer_3, properties {bLooping=true, soundWaveAssetPtr=&/DVR_Nav_Fast_Loop_Layer_03}}
]
properties {maxDistance=2097152, soundClassObject=&Fort_UI_InGame_NonCritical, volumeMultiplier=1.5}
