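# Sound cue for a looping altitude wind bed, apparently for the replay drone camera, routed to the Athena outdoor ambient sound class.
# A quality-level root picks between two looping wave players: the first two slots play the standard wind loop, the last two its OPT variant.
# The cue carries a reduced volume multiplier and a very large max distance.
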SoundCue /Game/Sounds/Fort_UI_Sounds/Replay/Drone_Altitude_Wind_Loop_Cue
root=QualityLevel_0
nodes [
 {id=QualityLevel_0, children [WavePlayer_0,WavePlayer_0,WavePlayer_1,WavePlayer_1]}
 {id=WavePlayer_0, properties {bLooping=true, soundWaveAssetPtr=&/Drone_Altitude_WindLoop}}
 {id=WavePlayer_1, properties {bLooping=true, soundWaveAssetPtr=&/Drone_Altitude_WindLoop_OPT}}
]
properties {maxDistance=2097152, soundClassObject=&Fort_Ambient_Outdoors_Athena, volumeMultiplier=0.576}
