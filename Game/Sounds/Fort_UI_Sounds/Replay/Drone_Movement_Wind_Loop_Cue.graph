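# Sound cue for a looping wind sound, named for drone movement among the replay UI sounds.
# A single looping wave player, set as the root, plays the Athena umbrella-closed falling wind loop.
# It plays at about half volume in the outdoor ambient sound class, with a very large maximum distance.
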SoundCue /Game/Sounds/Fort_UI_Sounds/Replay/Drone_Movement_Wind_Loop_Cue
root=WavePlayer_1
nodes [{id=WavePlayer_1, properties {bLooping=true, soundWaveAssetPtr=/Game/Athena/Sounds/Wind_Falling/Wind_UmbrellaClosed_Loop_01}}]
properties {maxDistance=2097152, soundClassObject=&Fort_Ambient_Outdoors_Athena, volumeMultiplier=0.501}
